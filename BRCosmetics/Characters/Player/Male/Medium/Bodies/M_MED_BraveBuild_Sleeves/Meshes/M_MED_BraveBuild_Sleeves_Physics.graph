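# Physics asset for the sleeved Brave Build male medium body cosmetic, driving secondary motion of its necktie and jacket.
# Dynamic capsule chains hang from small sphere anchor bodies: three necktie links, plus two-link jacket chains at front, sides and back.
# Constraints limit each link's swing, and most jacket chains lock twist.
# Pelvis, spine and thigh bodies form a coarse body volume, with only the spine linked by fully free constraints.
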PhysicsAsset /BRCosmetics/Characters/Player/Male/Medium/Bodies/M_MED_BraveBuild_Sleeves/Meshes/M_MED_BraveBuild_Sleeves_Physics
bodies {
 base_Necktie [{shapeType=Sphere, radius=0.505, shapeName=jacket_back_base_R_sphere}]
 dyn_Necktie01 [{shapeType=Capsule, length=4.25, radius=3, shapeName=dyn_Necktie01_capsule, center=<2.7246 0 -1.8245>, rotation=<rot 90 180 90>}]
 dyn_Necktie02 [{
   shapeType=Capsule
   length=5.05
   radius=4
   shapeName=dyn_Necktie02_capsule
   center=<2.2305 0.2134 -0.946>
   rotation=<rot 90 153.4349 -116.5651>
  }]
 dyn_Necktie03 [{shapeType=Capsule, length=3.9, radius=4.95, shapeName=dyn_Necktie03_capsule, center=<2.849 0 -2.0976>, rotation=<rot 90 180 -90>}]
 dyn_jacket_01_L [{
   shapeType=Capsule
   length=13
   radius=4
   shapeName=dyn_jacket_01_L_capsule
   center=<7.7159 1.0107 -4.1903>
   rotation=<rot -0.0003 -0.0014 75.9303>
  }]
 dyn_jacket_01_R [{
   shapeType=Capsule
   length=11
   radius=4.5
   shapeName=dyn_jacket_01_R_capsule
   center=<9.1698 -0.1749 -4.0207>
   rotation=<rot -0.0001 0.0003 110.0001>
  }]
 dyn_jacket_02_L [{
   shapeType=Capsule
   length=13
   radius=4.75
   shapeName=dyn_jacket_02_L_capsule
   center=<9.6224 0.1979 -3.9321>
   rotation=<rot -0.0001 -0.0004 80.8713>
  }]
 dyn_jacket_02_R [{
   shapeType=Capsule
   length=13
   radius=4.25
   shapeName=dyn_jacket_02_R_capsule
   center=<10.144 0.115 -2.8487>
   rotation=<rot -0.0001 0.0004 103.691>
  }]
 dyn_jacket_back_01 [{shapeType=Capsule, length=13, radius=4.75, shapeName=dyn_jacket_back_01_capsule, center=<9.1485 0 -3.6704>, rotation=<rot 0.0004 0 90>}]
 dyn_jacket_back_01_L [{
   shapeType=Capsule
   length=13
   radius=4.75
   shapeName=dyn_jacket_back_01_L_capsule
   center=<8.1888 0.1314 -5.534>
   rotation=<rot 0.0002 -0.0002 80>
  }]
 dyn_jacket_back_01_R [{
   shapeType=Capsule
   length=13
   radius=5
   shapeName=dyn_jacket_back_01_R_capsule
   center=<8.7203 -0.1149 -5.18>
   rotation=<rot 0.0002 0.0002 101.4091>
  }]
 dyn_jacket_back_02 [{shapeType=Capsule, length=13, radius=4.25, shapeName=dyn_jacket_back_02_capsule, center=<11.5746 0 -2.6542>, rotation=<rot 0.0004 0 90>}]
 dyn_jacket_back_02_L [{
   shapeType=Capsule
   length=12
   radius=4.5
   shapeName=dyn_jacket_back_02_L_capsule
   center=<10.1789 -0.8339 -5.458>
   rotation=<rot 0.0003 -0.0003 83.5679>
  }]
 dyn_jacket_back_02_R [{
   shapeType=Capsule
   length=11.25
   radius=5.5
   shapeName=dyn_jacket_back_02_R_capsule
   center=<10.6283 0.4099 -4.5181>
   rotation=<rot -0.5615 -3.005 104.06>
  }]
 dyn_jacket_front_01_L [{
   shapeType=Capsule
   length=15.75
   radius=3.6
   shapeName=dyn_jacket_front_01_L_capsule
   center=<7.8489 -1.483 -4.1104>
   rotation=<rot -0.0001 -0.0001 98.1418>
  }]
 dyn_jacket_front_01_R [{
   shapeType=Capsule
   length=13.7
   radius=4.05
   shapeName=dyn_jacket_front_01_R_capsule
   center=<7.2276 0.2496 -4.6129>
   rotation=<rot -0.0002 0.0002 80>
  }]
 dyn_jacket_front_02_L [{
   shapeType=Capsule
   length=12.25
   radius=8
   shapeName=dyn_jacket_front_02_L_capsule
   center=<8.3412 -0.8564 -6.5502>
   rotation=<rot -0.0006 -0.0004 97.9479>
  }]
 dyn_jacket_front_02_R [{
   shapeType=Capsule
   length=14.25
   radius=7.75
   shapeName=dyn_jacket_front_02_R_capsule
   center=<8.9486 1.3764 -8.2594>
   rotation=<rot -0.0003 0.0003 80>
  }]
 jacket_back_base [{shapeType=Sphere, radius=0.505, shapeName=jacket_back_base_sphere}]
 jacket_back_base_L [{shapeType=Sphere, radius=0.505, shapeName=jacket_back_base_L_sphere}]
 jacket_back_base_R [{shapeType=Sphere, radius=0.505, shapeName=jacket_back_base_R_sphere}]
 jacket_base_L [{shapeType=Sphere, radius=0.505, shapeName=jacket_base_L_sphere}]
 jacket_base_R [{shapeType=Sphere, radius=0.505, shapeName=jacket_base_R_sphere}]
 jacket_front_base_L [{shapeType=Sphere, radius=0.505, shapeName=jacket_front_base_L_sphere}]
 jacket_front_base_R [{shapeType=Sphere, radius=0.505, shapeName=jacket_front_base_R_sphere}]
 pelvis [{shapeType=Capsule, length=15.75, radius=9, shapeName=pelvis_capsule, center=<-4.1753 -3.5352 0>}]
 spine_01 [{shapeType=Capsule, length=20, radius=10, shapeName=spine_01_capsule, center=<-6.1655 1.8178 0>}]
 spine_02 [{shapeType=Sphere, radius=11, shapeName=spine_02_sphere}]
 spine_03 [{shapeType=Sphere, radius=11, shapeName=spine_03_sphere}]
 spine_04 [{shapeType=Sphere, radius=10.5, shapeName=spine_04_sphere}]
 spine_05 [{shapeType=Sphere, radius=12, shapeName=spine_05_sphere, center=<-3.4833 -1.7019 0>}]
 thigh_l [{shapeType=Capsule, length=47.0872, radius=9, center=<-18.2583 -0.6603 0.5553>, rotation=<rot 0 -88.8961 -88.4765>}]
 thigh_r [{shapeType=Capsule, length=42.1743, radius=8.6, center=<17.2028 1.4142 -0.2814>, rotation=<rot 0 -87.8875 -86.7104>}]
}
constraints [
 {bone1Name=dyn_jacket_back_01, bone2Name=jacket_back_base, swing1LimitDegrees=10, swing2LimitDegrees=90, twistLimitDegrees=10}
 {bone1Name=dyn_jacket_back_02, bone2Name=dyn_jacket_back_01, swing1LimitDegrees=10, swing2LimitDegrees=30, twistLimitDegrees=10}
 {
  bone1Name=dyn_jacket_back_01_L
  bone2Name=jacket_back_base_L
  swing1LimitDegrees=10
  swing2LimitDegrees=25
  twistLimitDegrees=10
  twistMotion=Locked
 }
 {
  bone1Name=dyn_jacket_back_02_L
  bone2Name=dyn_jacket_back_01_L
  swing1LimitDegrees=10
  swing2LimitDegrees=25
  twistLimitDegrees=10
  twistMotion=Locked
 }
 {
  bone1Name=dyn_jacket_back_01_R
  bone2Name=jacket_back_base_R
  swing1LimitDegrees=10
  swing2LimitDegrees=25
  twistLimitDegrees=10
  twistMotion=Locked
 }
 {
  bone1Name=dyn_jacket_back_02_R
  bone2Name=dyn_jacket_back_01_R
  swing1LimitDegrees=10
  swing2LimitDegrees=25
  twistLimitDegrees=10
  twistMotion=Locked
 }
 {bone1Name=dyn_jacket_01_L, bone2Name=jacket_base_L, swing1LimitDegrees=10, swing2LimitDegrees=25, twistLimitDegrees=10, twistMotion=Locked}
 {bone1Name=dyn_jacket_02_L, bone2Name=dyn_jacket_01_L, swing1LimitDegrees=10, swing2LimitDegrees=25, twistLimitDegrees=10, twistMotion=Locked}
 {bone1Name=dyn_jacket_01_R, bone2Name=jacket_base_R, swing1LimitDegrees=10, swing2LimitDegrees=25, twistLimitDegrees=10, twistMotion=Locked}
 {bone1Name=dyn_jacket_02_R, bone2Name=dyn_jacket_01_R, swing1LimitDegrees=10, swing2LimitDegrees=25, twistLimitDegrees=10, twistMotion=Locked}
 {
  bone1Name=dyn_jacket_front_01_L
  bone2Name=jacket_front_base_L
  swing1LimitDegrees=5
  swing2LimitDegrees=45
  twistLimitDegrees=10
  twistMotion=Locked
 }
 {
  bone1Name=dyn_jacket_front_02_L
  bone2Name=dyn_jacket_front_01_L
  swing1LimitDegrees=5
  swing2LimitDegrees=30
  twistLimitDegrees=10
  twistMotion=Locked
 }
 {
  bone1Name=dyn_jacket_front_01_R
  bone2Name=jacket_front_base_R
  swing1LimitDegrees=10
  swing2LimitDegrees=30
  twistLimitDegrees=10
  twistMotion=Locked
 }
 {
  bone1Name=dyn_jacket_front_02_R
  bone2Name=dyn_jacket_front_01_R
  swing1LimitDegrees=10
  swing2LimitDegrees=30
  twistLimitDegrees=1
  twistMotion=Locked
 }
 {bone1Name=base_Necktie, bone2Name=dyn_Necktie01, swing1LimitDegrees=10, swing2LimitDegrees=60, twistLimitDegrees=10}
 {bone1Name=dyn_Necktie01, bone2Name=dyn_Necktie02, swing1LimitDegrees=10, swing2LimitDegrees=55, twistLimitDegrees=10}
 {bone1Name=dyn_Necktie02, bone2Name=dyn_Necktie03, swing1LimitDegrees=10, swing2LimitDegrees=55, twistLimitDegrees=10}
 {
  bone1Name=spine_05
  bone2Name=spine_04
  swing1LimitDegrees=45
  swing1Motion=Free
  swing2LimitDegrees=45
  swing2Motion=Free
  twistLimitDegrees=45
  twistMotion=Free
 }
 {
  bone1Name=spine_02
  bone2Name=spine_01
  swing1LimitDegrees=45
  swing1Motion=Free
  swing2LimitDegrees=45
  swing2Motion=Free
  twistLimitDegrees=45
  twistMotion=Free
 }
 {
  bone1Name=spine_03
  bone2Name=spine_02
  swing1LimitDegrees=45
  swing1Motion=Free
  swing2LimitDegrees=45
  swing2Motion=Free
  twistLimitDegrees=45
  twistMotion=Free
 }
 {
  bone1Name=spine_04
  bone2Name=spine_03
  swing1LimitDegrees=45
  swing1Motion=Free
  swing2LimitDegrees=45
  swing2Motion=Free
  twistLimitDegrees=45
  twistMotion=Free
 }
]
