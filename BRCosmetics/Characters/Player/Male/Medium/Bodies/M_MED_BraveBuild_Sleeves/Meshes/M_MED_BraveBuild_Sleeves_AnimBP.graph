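# Post-process style cosmetic AnimBP for the BraveBuild Sleeves male body, animating its jacket and necktie on the player.
# The copied parent pose runs through Control Rig, leg-aware correctives, spine-relative RigidBody simulation, then movement-driven bone offsets and look-ats.
# Pelvis constraints partially pin jacket bones during skydive floating.
# Event logic computes those alphas, overrides gravity and drives material WPO cloth, enabled after a skydive delay and zeroed afterwards.
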AnimBlueprint /BRCosmetics/Characters/Player/Male/Medium/Bodies/M_MED_BraveBuild_Sleeves/Meshes/M_MED_BraveBuild_Sleeves_AnimBP
target_skeleton=&/M_MED_BraveBuild_Sleeves_Skeleton
graphs [
 AnimGraph {nodes [
   Copy Pose From Mesh() {bUseAttachedParent=true}
   "Transform (Modify) Bone - Bone: jacket_front_base_R"!(ComponentPose, Alpha=1, Scale="1,1,1") -> (Pose) {
    rotationMode=BMM_Additive
    rotationSpace=BCS_BoneSpace
    translationMode=BMM_Additive
    translationSpace=BCS_BoneSpace
    alphaBoolBlend {blendOption=Linear}
    boneToModify {boneName=jacket_front_base_R}
    rotation=<rot -15 0 0>
    translation {z=-2}
   }
   "Transform (Modify) Bone - Bone: dyn_jacket_front_02_L"(
    Alpha=Get LeftLegForwardBendCompensationAlpha() -> (LeftLegForwardBendCompensationAlpha:double)
    Translation="0,0,2"
    Rotation="0,0,0"
    Scale="1,1,1"
   ) {
    rotationMode=BMM_Additive
    rotationSpace=BCS_BoneSpace
    translationMode=BMM_Additive
    translationSpace=BCS_BoneSpace
    alphaBoolBlend {blendOption=Linear, blendOutTime=0.5}
    boneToModify {boneName=dyn_jacket_front_02_L}
    translation {z=2}
   }
   n1 = "Transform (Modify) Bone - Bone: dyn_jacket_front_02_L"!(ComponentPose, bAlphaBoolEnabled:bool=True, Scale="1,1,1") -> (Pose) {
    alphaInputType=Bool
    rotationMode=BMM_Additive
    rotationSpace=BCS_BoneSpace
    translationMode=BMM_Additive
    translationSpace=BCS_BoneSpace
    alphaBoolBlend {blendOption=Linear}
    alphaScaleBiasClamp {bInterpResult=true}
    boneToModify {boneName=dyn_jacket_front_02_L}
    translation {z=-3}
   }
   "Transform (Modify) Bone - Bone: dyn_jacket_front_02_R"!(ComponentPose=@n1, Alpha=1, Scale="1,1,1") -> (Pose) {
    rotationMode=BMM_Additive
    rotationSpace=BCS_BoneSpace
    translationMode=BMM_Additive
    translationSpace=BCS_BoneSpace
    alphaBoolBlend {blendOption=Linear}
    boneToModify {boneName=dyn_jacket_front_02_R}
    translation {z=-2}
   }
   "Transform (Modify) Bone - Bone: dyn_jacket_front_02_L"(bAlphaBoolEnabled=Get bIsNotCrouched() -> (bIsNotCrouched:bool), Scale="1,1,1") {
    alphaInputType=Bool
    rotationMode=BMM_Additive
    rotationSpace=BCS_BoneSpace
    translationMode=BMM_Additive
    translationSpace=BCS_BoneSpace
    alphaBoolBlend {blendOption=Linear}
    boneToModify {boneName=dyn_jacket_front_02_L}
    translation {z=-5}
   }
   n2 = Get FloatConstraint_alpha() -> (FloatConstraint_alpha:double)
   n3 = "Constraint - dyn_jacket_01_L (P)"(ComponentPose=Local To Component(LocalPose=Use cached pose 'PostRBAN'()), Alpha=@n2, ConstraintWeights_0=1) {
    alphaBoolBlend {blendOption=Linear}
    boneToModify {boneName=dyn_jacket_01_L}
    constraintSetup [{targetBone {boneName=pelvis}}]
    constraintWeights [1]
   }
   n4 = "Constraint - dyn_jacket_01_R (P)"(ComponentPose=@n3, Alpha=@n2, ConstraintWeights_0=1) {
    alphaBoolBlend {blendOption=Linear}
    boneToModify {boneName=dyn_jacket_01_R}
    constraintSetup [{targetBone {boneName=pelvis}}]
    constraintWeights [1]
   }
   n5 = "Constraint - dyn_jacket_front_01_L (P)"(ComponentPose=@n4, Alpha=@n2, ConstraintWeights_0=1) {
    alphaBoolBlend {blendOption=Linear}
    boneToModify {boneName=dyn_jacket_front_01_L}
    constraintSetup [{targetBone {boneName=pelvis}}]
    constraintWeights [1]
   }
   n6 = "Constraint - dyn_jacket_front_01_R (P)"(ComponentPose=@n5, Alpha=@n2, ConstraintWeights_0=1) {
    alphaBoolBlend {blendOption=Linear}
    boneToModify {boneName=dyn_jacket_front_01_R}
    constraintSetup [{targetBone {boneName=pelvis}}]
    constraintWeights [1]
   }
   n7 = "Constraint - dyn_jacket_back_01_L (P)"(ComponentPose=@n6, Alpha=@n2, ConstraintWeights_0=1) {
    alphaBoolBlend {blendOption=Linear}
    boneToModify {boneName=dyn_jacket_back_01_L}
    constraintSetup [{targetBone {boneName=pelvis}}]
    constraintWeights [1]
   }
   n8 = "Constraint - dyn_jacket_back_01_R (P)"(ComponentPose=@n7, Alpha=@n2, ConstraintWeights_0=1) {
    alphaBoolBlend {blendOption=Linear}
    boneToModify {boneName=dyn_jacket_back_01_R}
    constraintSetup [{targetBone {boneName=pelvis}}]
    constraintWeights [1]
   }
   n9 = "Constraint - dyn_jacket_back_01 (P)"(ComponentPose=@n8, Alpha=@n2, ConstraintWeights_0=1) {
    alphaBoolBlend {blendOption=Linear}
    boneToModify {boneName=dyn_jacket_back_01}
    constraintSetup [{targetBone {boneName=pelvis}}]
    constraintWeights [1]
   }
   n10 = "Constraint - dyn_jacket_02_L (P)"(ComponentPose=@n9, Alpha=@n2, ConstraintWeights_0=1) {
    alphaBoolBlend {blendOption=Linear}
    boneToModify {boneName=dyn_jacket_02_L}
    constraintSetup [{targetBone {boneName=pelvis}}]
    constraintWeights [1]
   }
   n11 = "Constraint - dyn_jacket_02_R (P)"(ComponentPose=@n10, Alpha=@n2, ConstraintWeights_0=1) {
    alphaBoolBlend {blendOption=Linear}
    boneToModify {boneName=dyn_jacket_02_R}
    constraintSetup [{targetBone {boneName=pelvis}}]
    constraintWeights [1]
   }
   n12 = "Constraint - dyn_jacket_front_02_L (P)"(ComponentPose=@n11, Alpha=@n2, ConstraintWeights_0=1) {
    alphaBoolBlend {blendOption=Linear}
    boneToModify {boneName=dyn_jacket_front_02_L}
    constraintSetup [{targetBone {boneName=pelvis}}]
    constraintWeights [1]
   }
   n13 = "Constraint - dyn_jacket_front_02_R (P)"(ComponentPose=@n12, Alpha=@n2, ConstraintWeights_0=1) {
    alphaBoolBlend {blendOption=Linear}
    boneToModify {boneName=dyn_jacket_front_02_R}
    constraintSetup [{targetBone {boneName=pelvis}}]
    constraintWeights [1]
   }
   n14 = "Constraint - dyn_jacket_back_02_L (P)"(ComponentPose=@n13, Alpha=@n2, ConstraintWeights_0=1) {
    alphaBoolBlend {blendOption=Linear}
    boneToModify {boneName=dyn_jacket_back_02_L}
    constraintSetup [{targetBone {boneName=pelvis}}]
    constraintWeights [1]
   }
   n15 = "Constraint - dyn_jacket_back_02_R (P)"(ComponentPose=@n14, Alpha=@n2, ConstraintWeights_0=1) {
    alphaBoolBlend {blendOption=Linear}
    boneToModify {boneName=dyn_jacket_back_02_R}
    constraintSetup [{targetBone {boneName=pelvis}}]
    constraintWeights [1]
   }
   n16 = "Constraint - dyn_jacket_back_02 (P)"(ComponentPose=@n15, Alpha=@n2, ConstraintWeights_0=1) {
    alphaBoolBlend {blendOption=Linear}
    boneToModify {boneName=dyn_jacket_back_02}
    constraintSetup [{targetBone {boneName=pelvis}}]
    constraintWeights [1]
   }
   Output Pose(Result=Component To Local(ComponentPose=@n16)) {layerGroup=None}
   n17 = "Transform (Modify) Bone - Bone: jacket_front_base_L"!(ComponentPose=Local To Component(LocalPose=Use cached pose 'RBAN'()), Alpha=1, Scale="1,1,1") -> (Pose) {
    rotationMode=BMM_Additive
    rotationSpace=BCS_BoneSpace
    translationMode=BMM_Additive
    translationSpace=BCS_BoneSpace
    alphaBoolBlend {blendOption=Linear}
    boneToModify {boneName=jacket_front_base_L}
    translation {x=0, z=-2}
   }
   n18 = "Transform (Modify) Bone - Bone: jacket_front_base_R"(ComponentPose=@n17, Alpha=1, Scale="1,1,1") {
    rotationMode=BMM_Additive
    rotationSpace=BCS_BoneSpace
    translationMode=BMM_Additive
    translationSpace=BCS_BoneSpace
    alphaBoolBlend {blendOption=Linear}
    boneToModify {boneName=jacket_front_base_R}
    translation {z=-1.5}
   }
   n19 = Get bIsCrouching()
   n20 = "Transform (Modify) Bone - Bone: dyn_jacket_front_02_R"!(ComponentPose=@n18, bAlphaBoolEnabled:bool=@n19, Scale="1,1,1") -> (Pose) {
    alphaInputType=Bool
    rotationMode=BMM_Additive
    rotationSpace=BCS_BoneSpace
    translationMode=BMM_Additive
    translationSpace=BCS_BoneSpace
    alphaBoolBlend {blendOption=Linear}
    boneToModify {boneName=dyn_jacket_front_02_R}
    rotation {pitch=-10}
    translation {z=-1}
   }
   n21 = "Transform (Modify) Bone - Bone: dyn_jacket_front_01_L"!(ComponentPose=@n20, bAlphaBoolEnabled:bool=True, Scale="1,1,1") -> (Pose) {
    alphaInputType=Bool
    rotationMode=BMM_Additive
    rotationSpace=BCS_BoneSpace
    translationMode=BMM_Additive
    translationSpace=BCS_BoneSpace
    alphaBoolBlend {blendOption=Linear}
    boneToModify {boneName=dyn_jacket_front_01_L}
    rotation {yaw=-5}
    translation {z=-1}
   }
   n22 = "Transform (Modify) Bone - Bone: dyn_jacket_front_02_L"(ComponentPose=@n21, bAlphaBoolEnabled=@n19, Scale="1,1,1") {
    alphaInputType=Bool
    rotationMode=BMM_Additive
    rotationSpace=BCS_BoneSpace
    translationMode=BMM_Additive
    translationSpace=BCS_BoneSpace
    alphaBoolBlend {blendOption=Linear}
    boneToModify {boneName=dyn_jacket_front_02_L}
    rotation {pitch=-10}
    translation {z=-3}
   }
   n23 = "Transform (Modify) Bone - Bone: dyn_jacket_front_02_R"!(
    ComponentPose=@n22
    bAlphaBoolEnabled=Get bIsMovingCrouched() -> (bIsMovingCrouched:bool)
    Translation:Vector="0,0,-1.3"
    Rotation:Rotator="0,0,0"
    Scale="1,1,1"
   ) -> (Pose) {
    alphaInputType=Bool
    rotationMode=BMM_Additive
    rotationSpace=BCS_BoneSpace
    translationMode=BMM_Additive
    translationSpace=BCS_BoneSpace
    alphaBoolBlend {blendOption=Linear, blendOutTime=0.5}
    alphaScaleBiasClamp {bInterpResult=true}
    boneToModify {boneName=dyn_jacket_front_02_R}
    translation {z=-1.3}
   }
   n24 = "Transform (Modify) Bone - Bone: dyn_jacket_front_01_R"!(
    ComponentPose=@n23
    bAlphaBoolEnabled:bool=Get bIsMovingRightCrouched() -> (bIsMovingRightCrouched:bool)
    Translation="0,0,0"
    Rotation="0,-15,0"
    Scale="1,1,1"
   ) -> (Pose) {
    alphaInputType=Bool
    rotationMode=BMM_Additive
    rotationSpace=BCS_BoneSpace
    translationMode=BMM_Additive
    translationSpace=BCS_BoneSpace
    alphaBoolBlend {blendOption=Linear, blendOutTime=0.5}
    alphaScaleBiasClamp {bInterpResult=true}
    boneToModify {boneName=dyn_jacket_front_01_R}
    rotation {yaw=-15}
   }
   n25 = "Transform (Modify) Bone - Bone: dyn_jacket_front_02_L"!(ComponentPose=@n24, bAlphaBoolEnabled=Get bIsMovingRightCrouched(), Translation:Vector="0,0,0", Rotation:Rotator="0,0,-13", Scale="1,1,1") -> (Pose) {
    alphaInputType=Bool
    rotationMode=BMM_Additive
    rotationSpace=BCS_BoneSpace
    translationMode=BMM_Additive
    translationSpace=BCS_BoneSpace
    alphaBoolBlend {blendOption=Linear, blendOutTime=0.5}
    alphaScaleBiasClamp {bInterpResult=true}
    boneToModify {boneName=dyn_jacket_front_02_L}
    rotation {roll=-13}
   }
   n26 = "Transform (Modify) Bone - Bone: dyn_jacket_02_R"!(ComponentPose=@n25, bAlphaBoolEnabled:bool=Get bIsMovingCrouched(), Translation="0,0,0", Rotation="-5,0,0", Scale="1,1,1") -> (Pose) {
    alphaInputType=Bool
    rotationMode=BMM_Additive
    rotationSpace=BCS_BoneSpace
    translationMode=BMM_Additive
    translationSpace=BCS_BoneSpace
    alphaBoolBlend {blendOption=Linear, blendOutTime=0.5}
    alphaScaleBiasClamp {bInterpResult=true}
    boneToModify {boneName=dyn_jacket_02_R}
    rotation {pitch=-5}
   }
   n27 = "Transform (Modify) Bone - Bone: jacket_front_base_L"!(ComponentPose=@n26, bAlphaBoolEnabled:bool=Get bIsMovingCrouched(), Translation:Vector="0,0,0", Rotation:Rotator="-5,0,0", Scale="1,1,1") -> (Pose) {
    alphaInputType=Bool
    rotationMode=BMM_Additive
    rotationSpace=BCS_BoneSpace
    translationMode=BMM_Additive
    translationSpace=BCS_BoneSpace
    alphaBoolBlend {blendOption=Linear, blendOutTime=0.5}
    alphaScaleBiasClamp {bInterpResult=true}
    boneToModify {boneName=jacket_front_base_L}
    rotation {pitch=-5}
   }
   n28 = "Transform (Modify) Bone - Bone: dyn_jacket_front_01_L"!(ComponentPose=@n27, bAlphaBoolEnabled=Get bIsMovingCrouched(), Translation:Vector="0,0,-1", Rotation:Rotator="0,0,0", Scale="1,1,1") -> (Pose) {
    alphaInputType=Bool
    rotationMode=BMM_Additive
    rotationSpace=BCS_BoneSpace
    translationMode=BMM_Additive
    translationSpace=BCS_BoneSpace
    alphaBoolBlend {blendOption=Linear, blendOutTime=0.5}
    alphaScaleBiasClamp {bInterpResult=true}
    boneToModify {boneName=dyn_jacket_front_01_L}
    translation {z=-1}
   }
   n29 = Get ForwardMovementAlpha() -> (ForwardMovementAlpha:double)
   n30 = "Transform (Modify) Bone - Bone: dyn_jacket_front_02_L"!(ComponentPose=@n28, Alpha:float=@n29, Translation="0,0,-2", Rotation="0,0,0", Scale="1,1,1") -> (Pose) {
    rotationMode=BMM_Additive
    rotationSpace=BCS_BoneSpace
    translationMode=BMM_Additive
    translationSpace=BCS_BoneSpace
    alphaBoolBlend {blendOption=Linear}
    alphaScaleBiasClamp {bInterpResult=true, interpSpeedDecreasing=20, interpSpeedIncreasing=20}
    boneToModify {boneName=dyn_jacket_front_02_L}
    translation {z=-2}
   }
   n31 = "Transform (Modify) Bone - Bone: jacket_front_base_R"!(ComponentPose=@n30, Alpha=@n29, Translation:Vector="0,0,-2", Rotation:Rotator="0,0,0", Scale="1,1,1") -> (Pose) {
    rotationMode=BMM_Additive
    rotationSpace=BCS_BoneSpace
    translationMode=BMM_Additive
    translationSpace=BCS_BoneSpace
    alphaBoolBlend {blendOption=Linear}
    alphaScaleBiasClamp {bInterpResult=true, interpSpeedDecreasing=20, interpSpeedIncreasing=20}
    boneToModify {boneName=jacket_front_base_R}
    rotation=<rot 0 0 0>
    translation {z=-2}
   }
   n32 = "Transform (Modify) Bone - Bone: dyn_Necktie02"(ComponentPose=@n31, Alpha=@n29, Translation="0,0,0", Rotation="-10,0,0", Scale="1,1,1") {
    rotationMode=BMM_Additive
    rotationSpace=BCS_BoneSpace
    translationMode=BMM_Additive
    translationSpace=BCS_BoneSpace
    alphaBoolBlend {blendOption=Linear}
    alphaScaleBiasClamp {bInterpResult=true, interpSpeedDecreasing=20, interpSpeedIncreasing=20}
    boneToModify {boneName=dyn_Necktie02}
    rotation {pitch=-10}
   }
   n33 = Get HeadRightUpAlpha() -> (HeadRightUpAlpha:float)
   n34 = "Transform (Modify) Bone - Bone: dyn_Necktie02"(ComponentPose=@n32, Alpha=@n33, Translation="0,0,-2", Rotation="0,0,0", Scale="1,1,1") {
    translationMode=BMM_Additive
    translationSpace=BCS_BoneSpace
    alphaBoolBlend {blendOption=Linear}
    boneToModify {boneName=dyn_Necktie02}
    translation {z=-2}
   }
   n35 = "Transform (Modify) Bone - Bone: dyn_Necktie03"(ComponentPose=@n34, Alpha=@n33, Translation="0,0,-2", Rotation="0,0,0", Scale="1,1,1") {
    translationMode=BMM_Additive
    translationSpace=BCS_BoneSpace
    alphaBoolBlend {blendOption=Linear}
    boneToModify {boneName=dyn_Necktie03}
    translation {z=-2}
   }
   n36 = Get bIsJumpingOrFalling() -> (bIsJumpingOrFalling:bool)
   n37 = "Transform (Modify) Bone - Bone: jacket_front_base_R"!(ComponentPose=@n35, bAlphaBoolEnabled:bool=@n36, Translation="0,0,0", Rotation="-5.000,0.000,-0.000", Scale="1,1,1") -> (Pose) {
    alphaInputType=Bool
    rotationMode=BMM_Additive
    rotationSpace=BCS_BoneSpace
    translationMode=BMM_Additive
    translationSpace=BCS_BoneSpace
    alphaBoolBlend {blendOption=Linear}
    alphaScaleBiasClamp {bInterpResult=true}
    boneToModify {boneName=jacket_front_base_R}
    rotation {pitch=-5}
   }
   n38 = "Transform (Modify) Bone - Bone: jacket_base_R"!(ComponentPose=@n37, bAlphaBoolEnabled:bool=@n36, Translation="0,0,0", Rotation="-5.000,0.000,-0.000", Scale="1,1,1") -> (Pose) {
    alphaInputType=Bool
    rotationMode=BMM_Additive
    rotationSpace=BCS_BoneSpace
    translationMode=BMM_Additive
    translationSpace=BCS_BoneSpace
    alphaBoolBlend {blendOption=Linear}
    alphaScaleBiasClamp {bInterpResult=true}
    boneToModify {boneName=jacket_base_R}
    rotation {pitch=-5}
   }
   n39 = "Transform (Modify) Bone - Bone: dyn_jacket_front_02_L"!(ComponentPose=@n38, bAlphaBoolEnabled:bool=@n36, Translation="0,0,-2", Rotation="0,0,0", Scale="1,1,1") -> (Pose) {
    alphaInputType=Bool
    rotationMode=BMM_Additive
    rotationSpace=BCS_BoneSpace
    translationMode=BMM_Additive
    translationSpace=BCS_BoneSpace
    alphaBoolBlend {blendOption=Linear, blendOutTime=1}
    alphaScaleBiasClamp {bInterpResult=true}
    boneToModify {boneName=dyn_jacket_front_02_L}
    translation {z=-2}
   }
   n40 = "Transform (Modify) Bone - Bone: dyn_jacket_front_02_R"(ComponentPose=@n39, bAlphaBoolEnabled=@n36, Translation="0,0,-1", Rotation="0,0,0", Scale="1,1,1") {
    alphaInputType=Bool
    rotationMode=BMM_Additive
    rotationSpace=BCS_BoneSpace
    translationMode=BMM_Additive
    translationSpace=BCS_BoneSpace
    alphaBoolBlend {blendOption=Linear, blendOutTime=1}
    alphaScaleBiasClamp {bInterpResult=true}
    boneToModify {boneName=dyn_jacket_front_02_R}
    translation {z=-1}
   }
   n41 = Get bIsMovingStanding() -> (bIsMovingStanding:bool)
   n42 = "Transform (Modify) Bone - Bone: dyn_jacket_front_01_L"(ComponentPose=@n40, bAlphaBoolEnabled=@n41, Translation="0,0,-2", Rotation="0,0,0", Scale="1,1,1") {
    alphaInputType=Bool
    rotationMode=BMM_Additive
    rotationSpace=BCS_BoneSpace
    translationMode=BMM_Additive
    translationSpace=BCS_BoneSpace
    alphaBoolBlend {blendInTime=0.5, blendOption=Linear, blendOutTime=1}
    alphaScaleBiasClamp {bInterpResult=true}
    boneToModify {boneName=dyn_jacket_front_01_L}
    translation {z=-2}
   }
   n43 = "Transform (Modify) Bone - Bone: dyn_jacket_front_01_R"(ComponentPose=@n42, bAlphaBoolEnabled=@n41, Translation="0,0,-2", Rotation="0,0,0", Scale="1,1,1") {
    alphaInputType=Bool
    rotationMode=BMM_Additive
    rotationSpace=BCS_BoneSpace
    translationMode=BMM_Additive
    translationSpace=BCS_BoneSpace
    alphaBoolBlend {blendInTime=0.5, blendOption=Linear, blendOutTime=1}
    alphaScaleBiasClamp {bInterpResult=true}
    boneToModify {boneName=dyn_jacket_front_01_R}
    translation {z=-2}
   }
   n44 = Get bIsMovingStandingForward() -> (bIsMovingStandingForward:bool)
   n45 = "Transform (Modify) Bone - Bone: dyn_jacket_front_01_L"(ComponentPose=@n43, bAlphaBoolEnabled=@n44, Translation="0,0,-3", Rotation="-15,0,0", Scale="1,1,1") {
    alphaInputType=Bool
    rotationMode=BMM_Additive
    rotationSpace=BCS_BoneSpace
    translationMode=BMM_Additive
    translationSpace=BCS_BoneSpace
    alphaBoolBlend {blendInTime=0.5, blendOption=Linear, blendOutTime=1}
    alphaScaleBiasClamp {bInterpResult=true}
    boneToModify {boneName=dyn_jacket_front_01_L}
    rotation {pitch=-15}
    translation {z=-3}
   }
   n46 = "Transform (Modify) Bone - Bone: dyn_jacket_front_01_R"(ComponentPose=@n45, bAlphaBoolEnabled=@n44, Translation="0,0,-1", Rotation="-10,0,0", Scale="1,1,1") {
    alphaInputType=Bool
    rotationMode=BMM_Additive
    rotationSpace=BCS_BoneSpace
    translationMode=BMM_Additive
    translationSpace=BCS_BoneSpace
    alphaBoolBlend {blendInTime=0.5, blendOption=Linear, blendOutTime=1}
    alphaScaleBiasClamp {bInterpResult=true}
    boneToModify {boneName=dyn_jacket_front_01_R}
    rotation {pitch=-10}
    translation {z=-1}
   }
   n47 = "Transform (Modify) Bone - Bone: dyn_jacket_front_02_L"!(ComponentPose=@n46, Alpha:float=1, Translation="0,0,-1", Rotation="0,0,0", Scale="1,1,1") -> (Pose) {
    rotationMode=BMM_Additive
    rotationSpace=BCS_BoneSpace
    translationMode=BMM_Additive
    translationSpace=BCS_BoneSpace
    alphaBoolBlend {blendOption=Linear, blendOutTime=1}
    alphaScaleBiasClamp {bInterpResult=true}
    boneToModify {boneName=dyn_jacket_front_02_L}
    translation {z=-1}
   }
   n48 = "Transform (Modify) Bone - Bone: dyn_jacket_front_02_R"!(ComponentPose=@n47, Alpha:float=1, Translation="0,0,-1", Rotation="0,0,0", Scale="1,1,1") -> (Pose) {
    rotationMode=BMM_Additive
    rotationSpace=BCS_BoneSpace
    translationMode=BMM_Additive
    translationSpace=BCS_BoneSpace
    alphaBoolBlend {blendOption=Linear, blendOutTime=1}
    alphaScaleBiasClamp {bInterpResult=true}
    boneToModify {boneName=dyn_jacket_front_02_R}
    translation {z=-1}
   }
   n49 = Look At - Bone: jacket_front_base_R(ComponentPose=@n48, Alpha=1) {
    alphaBoolBlend {blendOption=Linear}
    boneToModify {boneName=jacket_front_base_R}
    lookAtLocation=<27.4688 -13.9004 0>
    lookAtTarget {boneReference {boneName=thigh_r}}
    lookAt_Axis {axis=<1 0>}
   }
   n50 = Look At - Bone: jacket_front_base_L(ComponentPose=@n49, Alpha=1) {
    alphaBoolBlend {blendOption=Linear}
    boneToModify {boneName=jacket_front_base_L}
    lookAtLocation=<-20.3044 6.7008 0>
    lookAtTarget {boneReference {boneName=thigh_l}}
    lookAt_Axis {axis=<1 0>}
   }
   n51 = "Transform (Modify) Bone - Bone: dyn_jacket_front_02_R"(
    ComponentPose=@n50
    Alpha=Get RightLegCrouchCompensation() -> (RightLegCrouchCompensation:float)
    Translation="0,0,1"
    Rotation="0,0,0"
    Scale="1,1,1"
   ) {
    rotationMode=BMM_Additive
    rotationSpace=BCS_BoneSpace
    translationMode=BMM_Additive
    translationSpace=BCS_BoneSpace
    alphaBoolBlend {blendOption=Linear, blendOutTime=1}
    boneToModify {boneName=dyn_jacket_front_02_R}
    translation {z=1}
   }
   n52 = "Transform (Modify) Bone - Bone: dyn_Necktie01"!(ComponentPose=@n51, bAlphaBoolEnabled:bool=Get bIsSkyDiving(), Translation="0,0,-2", Rotation="0,0,0", Scale="1,1,1") -> (Pose) {
    alphaInputType=Bool
    translationMode=BMM_Additive
    translationSpace=BCS_BoneSpace
    alphaBoolBlend {blendOption=Linear}
    boneToModify {boneName=dyn_Necktie01}
    translation {z=-2}
   }
   n53 = Look At - Bone: dyn_jacket_front_02_R(ComponentPose=@n52, Alpha=Get RightLegCrouchCompensation()) {
    alphaBoolBlend {blendOption=Linear}
    boneToModify {boneName=dyn_jacket_front_02_R}
    lookAtLocation=<0 -10>
    lookAtTarget {boneReference {boneName=calf_r}}
    lookAt_Axis {axis=<1 0>}
   }
   n54 = "Transform (Modify) Bone - Bone: dyn_jacket_front_02_R"!(ComponentPose=@n53, bAlphaBoolEnabled:bool=Get bIsMovingCrouched(), Translation="0,0,-1", Rotation="0,0,0", Scale="1,1,1") -> (Pose) {
    alphaInputType=Bool
    rotationMode=BMM_Additive
    rotationSpace=BCS_BoneSpace
    translationMode=BMM_Additive
    translationSpace=BCS_BoneSpace
    alphaBoolBlend {blendOption=Linear, blendOutTime=1}
    boneToModify {boneName=dyn_jacket_front_02_R}
    translation {z=-1}
   }
   Save cached pose 'PostRBAN'(Pose=Component To Local(ComponentPose=@n54)) {cacheName=PostRBAN}
   n55 = Control Rig(Source=Copy Pose From Mesh() {bUseAttachedParent=true}) {
    alphaBoolBlend {blendOption=Linear}
    controlRigAssetReference {blueprintRigClass=&/M_MED_BraveBuild_Sleeves_ControlRig.M_MED_BraveBuild_Sleeves_ControlRig_C}
   }
   n56 = "Transform (Modify) Bone - Bone: jacket_front_base_L"!(ComponentPose=Local To Component(LocalPose=@n55), bAlphaBoolEnabled=Get bIsCrouching(), Scale="1,1,1") -> (Pose) {
    alphaInputType=Bool
    rotationMode=BMM_Additive
    rotationSpace=BCS_BoneSpace
    translationMode=BMM_Additive
    translationSpace=BCS_BoneSpace
    alphaBoolBlend {blendOption=Linear}
    alphaScaleBiasClamp {bInterpResult=true}
    boneToModify {boneName=jacket_front_base_L}
    rotation {pitch=-74.8685, yaw=-10}
    translation {x=-5, z=-5}
   }
   n57 = "Transform (Modify) Bone - Bone: jacket_base_L"!(ComponentPose=@n56, bAlphaBoolEnabled:bool=Get bIsCrouching(), Translation="0,0,-2", Rotation="0,-15,0", Scale="1,1,1") -> (Pose) {
    alphaInputType=Bool
    rotationMode=BMM_Additive
    rotationSpace=BCS_BoneSpace
    translationMode=BMM_Additive
    translationSpace=BCS_BoneSpace
    alphaBoolBlend {blendOption=Linear}
    alphaScaleBiasClamp {bInterpResult=true}
    boneToModify {boneName=jacket_base_L}
    rotation {yaw=-15}
    translation {z=-2}
   }
   n58 = "Transform (Modify) Bone - Bone: jacket_front_base_L"!(
    ComponentPose=@n57
    Alpha:float=Get LeftLegForwardBendCompensationAlpha()
    Translation:Vector="0,0,0"
    Rotation:Rotator="-15,0,-15"
    Scale="1,1,1"
   ) -> (Pose) {
    rotationMode=BMM_Additive
    rotationSpace=BCS_BoneSpace
    translationMode=BMM_Additive
    translationSpace=BCS_BoneSpace
    alphaBoolBlend {blendOption=Linear}
    alphaScaleBiasClamp {bInterpResult=true}
    boneToModify {boneName=jacket_front_base_L}
    rotation {pitch=-15, roll=-15}
   }
   n59 = Look At - Bone: jacket_front_base_R(ComponentPose=@n58, Alpha=1) {
    alphaBoolBlend {blendOption=Linear}
    boneToModify {boneName=jacket_front_base_R}
    lookAtLocation=<0 -10>
    lookAtTarget {boneReference {boneName=calf_r}}
    lookAt_Axis {axis=<1 0>}
   }
   n60 = Look At - Bone: jacket_base_R(ComponentPose=@n59, Alpha=0.7) {
    alphaBoolBlend {blendOption=Linear}
    boneToModify {boneName=jacket_base_R}
    lookAtLocation=<0 -10>
    lookAtTarget {boneReference {boneName=calf_r}}
    lookAt_Axis {axis=<1 0>}
   }
   Save cached pose 'PreRBAN'(Pose=Component To Local(ComponentPose=@n60)) {cacheName=PreRBAN}
   n61 = RigidBody!(
    ComponentPose=Local To Component(LocalPose=Use cached pose 'PreRBAN'())
    Alpha
    OverrideWorldGravity:Vector=Get Gravity Override()
    ExternalForce
   ) -> (Pose) {
    bClampLinearTranslationLimitToRefPose=true
    bOverrideWorldGravity=true
    overridePhysicsAsset=&/M_MED_BraveBuild_Sleeves_Physics
    simulationSpace=BaseBoneSpace
    alphaBoolBlend {blendOption=Linear}
    alphaScaleBiasClamp {inRange {max=0.033, min=0.016}, outRange {max=0.7, min=1}}
    baseBoneRef {boneName=spine_01}
    componentAppliedLinearAccClamp=<2000 2000 2000>
    componentLinearAccScale=<0.2 0.2 0.2>
    componentLinearVelScale=<0.2 0.2 0.2>
    overrideWorldGravity {z=-980}
   }
   Save cached pose 'RBAN'(Pose=Component To Local(ComponentPose=@n61)) {cacheName=RBAN}
   Get bIsPlayingEmote()
   Get Delta Time X()
  ]}
 DisableWPO {nodes [
   DisableWPO() {is=title K2Node_FunctionEntry}
   > n1 = Sequence()
   n2 = Get Mat Instance Dynamic()
   > SetScalarParameterValue(self=@n2, ParameterName=ClothSim_R1_DragStrength, Value=0)
   > SetScalarParameterValue(self=@n2, ParameterName=ClothSim_R1_TurbulenceStrength, Value=0)
   > SetScalarParameterValue(self=@n2, ParameterName=ClothSim_R1_BoneRadius, Value=0)
   > SetScalarParameterValue(self=@n2, ParameterName=ClothSim_R1_BoneThreshold, Value=0)
   SetScalarParameterValue(execute=@n1.then_1, self=@n2, ParameterName=ClothSim_R2_DragStrength, Value=0)
   > SetScalarParameterValue(self=@n2, ParameterName=ClothSim_R2_TurbulenceStrength, Value=0)
   > SetScalarParameterValue(self=@n2, ParameterName=ClothSim_R2_BoneRadius, Value=0)
   > SetScalarParameterValue(self=@n2, ParameterName=ClothSim_R2_BoneThreshold, Value=0)
  ]}
 EnableWPO {nodes [
   EnableWPO() {is=title K2Node_FunctionEntry}
   > n1 = Sequence()
   n2 = Get Mat Instance Dynamic()
   n3 = Get!() -> (WPO_R1_Default_X:double, WPO_R1_Default_Y:double, WPO_R1_Default_Z:double, WPO_R1_Default_W:double)
   > SetScalarParameterValue(self=@n2, ParameterName=ClothSim_R1_DragStrength, Value=@n3)
   > SetScalarParameterValue(self=@n2, ParameterName=ClothSim_R1_TurbulenceStrength, Value=@n3.WPO_R1_Default_Y)
   > SetScalarParameterValue(self=@n2, ParameterName=ClothSim_R1_BoneRadius, Value=@n3.WPO_R1_Default_Z)
   > SetScalarParameterValue(self=@n2, ParameterName=ClothSim_R1_BoneThreshold, Value=@n3.WPO_R1_Default_W)
   n4 = Get!() -> (WPO_R2_Default_X:double, WPO_R2_Default_Y:double, WPO_R2_Default_Z:double, WPO_R2_Default_W:double)
   SetScalarParameterValue(execute=@n1.then_1, self=@n2, ParameterName=ClothSim_R2_DragStrength, Value=@n4)
   > SetScalarParameterValue(self=@n2, ParameterName=ClothSim_R2_TurbulenceStrength, Value=@n4.WPO_R2_Default_Y)
   > SetScalarParameterValue(self=@n2, ParameterName=ClothSim_R2_BoneRadius, Value=@n4.WPO_R2_Default_Z)
   > SetScalarParameterValue(self=@n2, ParameterName=ClothSim_R2_BoneThreshold, Value=@n4.WPO_R2_Default_W)
  ]}
 EventGraph {nodes [
   Event OnCharacterCustomizationLoaded()
   > n1 = Set Skel Mesh(Skel Mesh=GetPartSkeletalMeshComponent())
   > n2 = Get Dynamic Material Instance(SkeletalMeshComponent=@n1)
   > Set Mat Instance Dynamic(Mat Instance Dynamic=@n2)
   > n3 = GetOrCreateDynamicOverlayMaterial(MeshComponent=@n1, MaterialName=None)
   > Set Overlay Mat Instance Dynamic(Overlay Mat Instance Dynamic=@n3)
   > n4 = Sequence()
   > Do Once()
   > SaveWPODefaults()
   DisableWPO(execute=@n4.then_1) {is=title K2Node_CallFunction}
   Event BlueprintInitializeAnimation()
   > Set Skel Mesh(Skel Mesh=GetOwnerSkeletalMeshComponent())
   n5 = TryGetPawnOwner()
   > Is Valid(InputObject=@n5)
   > n6 = Cast To FortPlayerPawnAthena(Object=@n5)
   > Set PlayerPawn(PlayerPawn:Fort Player Pawn Athena Object Reference=@n6) -> (Output_Get:Fort Player Pawn Athena Object Reference)
   Event OnBeginSkydiving()
   > RetriggerableDelay(Duration=3)
   > EnableWPO() {is=title K2Node_CallFunction}
   n7 = Event BlueprintUpdateAnimation()
   > n8 = Set Delta Time X(Delta Time X=@n7.DeltaTimeX)
   > Set FPS Compensation Alpha(FPS Compensation Alpha:double=MapRangeClamped(Value=@n8, InRangeA=0.02, InRangeB=0.03)) -> (Output_Get:double)
   > n9 = Sequence!(execute) -> (then_0, then_1, then_2:Exec, then_3:Exec, then_4:Exec)
   n10 = Get Skel Mesh()
   > n11 = Set LeftLegForwardBendCompensationAlpha(LeftLegForwardBendCompensationAlpha:double=GetAlphaFromReferenceTransformDelta(SkeletalMeshComponent=@n10, BoneName=thigh_l, BaseName=pelvis, InputRange=(X=-17,Y=-60), TwistAxis="1,0,0")) -> (Output_Get:double)
   > Set EMOTE_LeftThighBendFPSCompensationAlpha(EMOTE_LeftThighBendFPSCompensationAlpha:double=float * float!(A=@n11, B=Get FPS Compensation Alpha() -> (FPS Compensation Alpha:double), C:bool=Get bIsPlayingEmote()) -> (ReturnValue) {is=title K2Node_PromotableOperator}) -> (Output_Get:double)
   > n12 = Set RightLegForwardBendCompensationAlpha(RightLegForwardBendCompensationAlpha:double=GetAlphaFromReferenceTransformDelta(SkeletalMeshComponent=@n10, BoneName=thigh_r, BaseName=pelvis, InputRange=(X=-17,Y=-60), TwistAxis="1,0,0")) -> (Output_Get:double)
   > Set EMOTE_RightThighBendFPSCompensationAlpha(EMOTE_RightThighBendFPSCompensationAlpha:double=float * float(A=@n12, B=Get FPS Compensation Alpha(), C=Get bIsPlayingEmote())) -> (Output_Get:double)
   > Set bIsJumpingOrFalling(bIsJumpingOrFalling:bool=OR Boolean(A=Get bIsJumping(), B=Get bIsFalling())) -> (Output_Get:bool)
   > Set HeadRightUpAlpha(HeadRightUpAlpha:float=GetAlphaFromReferenceTransformDelta(SkeletalMeshComponent=Get Skel Mesh(), BoneName=head, BaseName=spine_05, InputRange=(X=0,Y=-20), TwistAxis="1,0,0")) -> (Output_Get:float)
   > Set bIsMovingCrouched(bIsMovingCrouched:bool=AND Boolean(A=Get bIsMoving2D(), B=Get bIsCrouching())) -> (Output_Get:bool)
   n13 = Get Skel Mesh()
   > Set RightLegCrouchCompensation(RightLegCrouchCompensation:float=SelectFloat(
     A=GetAlphaFromReferenceTransformDelta(SkeletalMeshComponent=@n13, BoneName=thigh_r, BaseName=pelvis, InputRange=(X=-85,Y=-100), TwistAxis="1,0,0")
     B=0
     bPickA=Get bIsCrouching()
    )) -> (Output_Get:float)
   n14 = Get bIsCrouching()
   > Set bIsMovingStanding(bIsMovingStanding:bool=AND Boolean(A=Get bIsMoving2D(), B=NOT Boolean(A=@n14))) -> (Output_Get:bool)
   > Set bIsNotCrouched(bIsNotCrouched:bool=NOT Boolean(A=Get bIsCrouching())) -> (Output_Get:bool)
   n15 = Get PawnMovementDirectionAzimuth() -> (PawnMovementDirectionAzimuth:float)
   > Set bIsMovingStandingForward(bIsMovingStandingForward:bool=AND Boolean!(A=Get bIsMovingStanding(), B="float > float"(A=@n15, B=-45), C:bool="float < float"(A=@n15, B=45)) -> (ReturnValue)) -> (Output_Get:bool)
   n16 = Get PawnMovementDirectionAzimuth()
   > Set bIsMovingRightCrouched(bIsMovingRightCrouched:bool=AND Boolean!(A="float > float"(A=@n16, B=45), B="float < float"(A=@n16, B=135), C=Get bIsCrouching(), D:bool=Get bIsMoving2D()) -> (ReturnValue)) -> (Output_Get:bool)
   Is Valid(exec=@n9.then_3, InputObject=Get PlayerPawn() -> (PlayerPawn:Fort Player Pawn Athena Object Reference))
   > Set WPO Globals Function(
    Material=Get Mat Instance Dynamic()
    Overlay Material=Get Overlay Mat Instance Dynamic()
    Skeletal Mesh=Get Skel Mesh()
    "Local Noise Transform  Location"="Get Local Noise Transform  Location"() -> ("Local Noise Transform  Location":Vector)
    Local Noise Offset=Get Local Noise Offset()
   )
   > Set WPO Regional Function!(
    execute
    Material:Material Instance Dynamic Object Reference=Get Mat Instance Dynamic()
    Overlay Material:Material Instance Dynamic Object Reference=Get Overlay Mat Instance Dynamic()
    Skeletal Mesh:Skeletal Mesh Component Object Reference=Get Skel Mesh()
    Velocity Transform:Name=dyn_Necktie01
    Velocity Interp Speed:double=Get Velocity Interp Speed()
    Collision Transform:Name=dyn_Necktie03
    Collision Transform Axis:EAxis Enum
    Invert Collision Transform Axis:bool=true
    Delta Time X:double=0
    Previous Velocity Transform Location:"Vector (by ref)"=Get R1 Velocity Transform Location()
    Previous Velocity Transform Velocity:"Vector (by ref)"=Get R1 Velocity Transform Velocity()
    Region:int=1
   ) -> (then)
   > Set WPO Regional Function(
    Material=Get Mat Instance Dynamic()
    Overlay Material=Get Overlay Mat Instance Dynamic()
    Skeletal Mesh=Get Skel Mesh()
    Velocity Transform=pelvis
    Velocity Interp Speed=Get Velocity Interp Speed()
    Collision Transform=spine_01
    Invert Collision Transform Axis=false
    Delta Time X=0
    Previous Velocity Transform Location=Get R2 Velocity Transform Location()
    Previous Velocity Transform Velocity=Get R2 Velocity Transform Velocity()
    Region=2
   )
   > SetScalarParameterValue(self=Get Mat Instance Dynamic(), ParameterName=ClothSim_WindStrength, Value=SelectFloat(A=1, B=0, bPickA=Get bIsSkydiveFloating()))
   n17 = Gravity Override State Function DEPRECATED(
    execute=@n9.then_2
    CharacterPartAnimInstance=Self-Reference()
    GameTimeInSeconds=GetGameTimeInSeconds()
    GravityOverrideParams=Get Gravity Override Params()
   )
   > Set Gravity Override(Gravity Override=@n17)
   n18 = TryGetPawnOwner()
   n19 = GetVelocity!(self:Actor Object Reference=@n18) -> (ReturnValue:Vector)
   n20 = GetActorForwardVector!(self:Actor Object Reference=@n18) -> (ReturnValue:Vector)
   Set ForwardMovementAlpha(
    execute=@n9.then_1
    ForwardMovementAlpha:double=MapRangeClamped(Value=Dot Product!(A:Vector=@n19, B:Vector=@n20) -> (ReturnValue:double), InRangeA=0, InRangeB=400)
   ) -> (Output_Get:double)
   > Set FloatConstraint_alpha(FloatConstraint_alpha:double=SelectFloat(A=0.4, B=0, bPickA=Get bIsSkydiveFloating())) -> (Output_Get:double)
   Event OnEndSkydiving()
   > DisableWPO()
  ]}
 SaveWPODefaults {nodes [
   SaveWPODefaults() {is=title K2Node_FunctionEntry}
   > n1 = Sequence!(execute) -> (then_0, then_1, then_2)
   n2 = Get Mat Instance Dynamic()
   > n3 = Get Scalar Parameter Value!(execute, self:Material Instance Dynamic Object Reference=@n2, ParameterName:Name=ClothSim_R1_DragStrength) -> (then, ReturnValue:float) {is=title K2Node_CallFunction}
   n4 = Get Mat Instance Dynamic()
   > n5 = Get Scalar Parameter Value(self=@n4, ParameterName=ClothSim_R1_TurbulenceStrength)
   n6 = Get Mat Instance Dynamic()
   > n7 = Get Scalar Parameter Value(self=@n6, ParameterName=ClothSim_R1_BoneRadius)
   n8 = Get Mat Instance Dynamic()
   > n9 = Get Scalar Parameter Value(self=@n8, ParameterName=ClothSim_R1_BoneThreshold)
   > Set!(execute, WPO_R1_Default_X:double=@n3, WPO_R1_Default_Y:double=@n5, WPO_R1_Default_Z:double=@n7, WPO_R1_Default_W:double=@n9) -> (then, Output_Get:Vector 4 Structure)
   n10 = Get Scalar Parameter Value(execute=@n1.then_1, self=@n2, ParameterName=ClothSim_R2_DragStrength)
   > n11 = Get Scalar Parameter Value(self=@n4, ParameterName=ClothSim_R2_TurbulenceStrength)
   > n12 = Get Scalar Parameter Value(self=@n6, ParameterName=ClothSim_R2_BoneRadius)
   > n13 = Get Scalar Parameter Value(self=@n8, ParameterName=ClothSim_R2_BoneThreshold)
   > Set!(execute, WPO_R2_Default_X:double=@n10, WPO_R2_Default_Y:double=@n11, WPO_R2_Default_Z:double=@n12, WPO_R2_Default_W:double=@n13) -> (then, Output_Get)
  ]}
]
variables [
 bIsJumpingOrFalling {type=boolean}
 bIsMovingCrouched {type=boolean}
 bIsMovingRightCrouched {type=boolean}
 bIsMovingStanding {type=boolean}
 bIsMovingStandingForward {type=boolean}
 bIsNotCrouched {type=boolean}
 Delta Time X {type=number}
 EMOTE_LeftThighBendFPSCompensationAlpha {type=number}
 EMOTE_RightThighBendFPSCompensationAlpha {type=number}
 FloatConstraint_alpha {type=number}
 ForwardMovementAlpha {type=number}
 FPS Compensation Alpha {type=number}
 Gravity Override {type=Vector}
 Gravity Override Params {type=GravityOverrideParamsStruct}
 HeadRightUpAlpha {type=number}
 LeftLegForwardBendCompensationAlpha {type=number}
 Local Noise Offset {type=Vector}
 "Local Noise Transform  Location" {type=Vector}
 Mat Instance Dynamic {type=MaterialInstanceDynamic}
 NewVar {type=boolean}
 Overlay Mat Instance Dynamic {type=MaterialInstanceDynamic}
 PlayerPawn {type=FortPlayerPawnAthena}
 R1 Velocity Transform Location {type=Vector}
 R1 Velocity Transform Velocity {type=Vector}
 R2 Velocity Transform Location {type=Vector}
 R2 Velocity Transform Velocity {type=Vector}
 RightLegCrouchCompensation {type=number}
 RightLegForwardBendCompensationAlpha {type=number}
 Skel Mesh {type=SkeletalMeshComponent}
 Velocity Interp Speed {type=number}
 WPO Pho Wind Intensity {type=number}
 WPO_R1_Default {type=Vector4, default=<0 0 0 0>}
 WPO_R2_Default {type=Vector4, default=<0 0 0 0>}
]
